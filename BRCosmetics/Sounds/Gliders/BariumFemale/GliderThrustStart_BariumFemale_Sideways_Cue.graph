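# Sound cue for the start of a sideways thrust on the BariumFemale glider cosmetic.
# A Random node picks one of four equally weighted thrust waves and feeds a Modulator that varies playback, with a lowered pitch minimum.
# The cue plays at reduced volume in the glider landing sound class, using the shared player glider thrust attenuation and glider thrust concurrency assets.
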SoundCue /BRCosmetics/Sounds/Gliders/BariumFemale/GliderThrustStart_BariumFemale_Sideways_Cue
root=Modulator_0
nodes [
 {id=Modulator_0, children [Random_0], properties {pitchMin=0.9}}
 {id=Random_0, children [WavePlayer_0,WavePlayer_1,WavePlayer_2,WavePlayer_3], properties {weights [1,1,1,1]}}
 {id=WavePlayer_0, properties {soundWaveAssetPtr=&/Glider_Barium_Sideways_Thrust_01}}
 {id=WavePlayer_1, properties {soundWaveAssetPtr=&/Glider_Barium_Sideways_Thrust_02}}
 {id=WavePlayer_2, properties {soundWaveAssetPtr=&/Glider_Barium_Sideways_Thrust_03}}
 {id=WavePlayer_3, properties {soundWaveAssetPtr=&/Glider_Barium_Sideways_Thrust_04}}
]
properties {
 attenuationSettings=&Fort_Player_Glider_Thrust_Athena_Att
 maxDistance=12500
 soundClassObject=&Fort_Gliders_Land
 volumeMultiplier=0.7
 concurrencyOverrides {bLimitToOwner=true, maxCount=1, resolutionRule=PreventNew, voiceStealReleaseTime=0.25}
 concurrencySet [&Athena_Glider_Thrust_Conc]
}
